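# Translucent deferred decal material for the Scrub A Dub emote, drawing a wet water patch with bubbly foam.
# Tiled blurred noise times a radial gradient, offset by erosion parameters, forms sharpened water and foam masks, with a bubble texture breaking up the foam.
# Named reroutes carry the masks and decal lifetime opacity into blends for color, opacity, specular, roughness and bubble normal.
# A facing mask borrowed from the SprayGun prototype also scales opacity.
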Material /FNGameplayCues/Animation/Game/MainPlayer/Emotes/Scrub_A_Dub/VFX/M_Emote_ScrubADub_Wet_Decal
graph {
 properties {
  bCastRayTracedShadows=false
  blendMode=BLEND_Translucent
  materialDomain=MD_DeferredDecal
  pixelDepthOffsetMode=PDOM_Legacy
  usedShadingModels=Default Lit
 }
 nodes [
  Constant()
  n1 = NamedRerouteUsage()
  DotProduct(A=@n1)
  Fresnel(Normal=Transform(None=@n1)) {baseReflectFraction=0.02, exponent=4}
  n2 = VectorParameter() {parameterName=Water Foam Erosion, defaultValue=<rgba -0.55 -0.05 -0.65 -0.125>}
  n3 = Multiply(A=TextureCoordinate(), B=ScalarParameter() {defaultValue=2.8, parameterName=Foam Tile})
  n4 = MaterialFunctionCall() {materialFunction=&RadialGradientExponential}
  n5 = TextureCoordinate() {uTiling=5, vTiling=5}
  n6 = TextureSampleParameter2D(UVs=@n5) {parameterName=Bubble Mask, samplerType=SAMPLERTYPE_LinearGrayscale, texture=/Game/Global/Textures/Default/Noise/T_Bubble_B}
  n7 = Add(
   A=LinearInterpolate(A=@n2.B, B=@n2.B, Alpha=NamedRerouteUsage()) {constA=-1, constB=-0.2}
   B=Multiply(
    A=TextureSampleParameter2D(UVs=@n3) {parameterName=Foam Mask, samplerType=SAMPLERTYPE_Grayscale, texture=&LowResBlurredNoise}.R
    B=Multiply(A=@n4.RadialGradientExponential, B=@n6.R)
   )
  )
  NamedRerouteDeclaration(None=Saturate(None=Divide(A=@n7) {constB=0.05})) {name=Foam Mask, nodeColor=<rgba 1 0 0.8906 1>}
  n8 = VectorParameter() {parameterName=Base Color, defaultValue=<rgba 0 0.2917 0.5 0.1>}
  n9 = VectorParameter() {parameterName=Foam Color, defaultValue=<rgba 0.67 0.89 1 0.5>}
  n10 = NamedRerouteUsage()
  NamedRerouteDeclaration(None=LinearInterpolate(A=@n8.RGB, B=@n9.RGB, Alpha=@n10) {constAlpha=0}) {name=Diffuse, nodeColor=<rgba 0 0.5078 1 1>}
  n11 = Multiply(A=@n8.A, B=NamedRerouteUsage())
  NamedRerouteDeclaration(None=Multiply(
    A=LinearInterpolate(A=@n11, B=@n9.A, Alpha=@n10)
    B=MaterialFunctionCall() {materialFunction=/Game/Weapons/Prototype/SprayGun/Materials/MF_DecalFacingMask}.Result
   )) {name=Opacity, nodeColor=<rgba 0 0.2969 1 1>}
  NamedRerouteDeclaration(None=LinearInterpolate(
    A=ScalarParameter() {defaultValue=1, parameterName=Specular - Water}
    B=ScalarParameter() {defaultValue=1, parameterName=Specular - Foam}
    Alpha=NamedRerouteUsage()
   ) {constA=1, constB=0.5}) {name=Specular, nodeColor=<rgba 0.3594 1 0 1>}
  NamedRerouteDeclaration(None=LinearInterpolate(
    A=ScalarParameter() {defaultValue=0.2, parameterName=Roughness - Water}
    B=ScalarParameter() {parameterName=Roughness - Foam}
    Alpha=NamedRerouteUsage()
   ) {constA=0.2, constB=0.8}) {name=Roughness, nodeColor=<rgba 0.0312 1 0 1>}
  n12 = NamedRerouteUsage()
  n13 = Multiply(A=TextureCoordinate(), B=ScalarParameter() {defaultValue=1.5, parameterName=Water Tile})
  n14 = Add(
   A=LinearInterpolate(A=@n2.R, B=@n2.R, Alpha=SmoothStep(Value=@n12) {constMax=0.66}) {constA=-1, constAlpha=1, constB=0}
   B=Multiply(
    A=TextureSampleParameter2D(UVs=@n13) {parameterName=Foam Mask, samplerType=SAMPLERTYPE_Grayscale, texture=&LowResBlurredNoise}.R
    B=@n4.RadialGradientExponential
   )
  )
  NamedRerouteDeclaration(None=Saturate(None=Divide(A=@n14) {constB=0.1})) {name=Water Mask, nodeColor=<rgba 0.0547 0 1 1>}
  NamedRerouteDeclaration(None=DecalLifetimeOpacity().Opacity) {name=Decal Opacity, nodeColor=<rgba 1 0 0.4922 1>}
  n15 = TextureSampleParameter2D(UVs=@n5) {parameterName=Foam, samplerType=SAMPLERTYPE_Normal, texture=/Game/Global/Textures/Default/Noise/T_Bubble_B_N}
  NamedRerouteDeclaration(None=LinearInterpolate(A=Constant3Vector() {constant=<rgba 0 0 1 0>}, B=@n15.RGB, Alpha=NamedRerouteUsage())) {name=Normal, nodeColor=<rgba 1 0.8438 0 1>}
  NamedRerouteDeclaration(None=@n15.RGB) {name=Normal_L1, nodeColor=<rgba 0.7578 1 0 1>}
  NamedRerouteUsage()
  NamedRerouteUsage()
  NamedRerouteUsage()
  NamedRerouteUsage()
  NamedRerouteUsage()
  Normalize(VectorInput=Multiply(A=VectorParameter() {parameterName=Normal Mult, defaultValue=<rgba 2 2 1 0>}.RGB, B=@n15.RGB))
  n16 = Time()
  Saturate(None=ConstantBiasScale(None=Sine(None=@n16) {period=4}) {scale=1})
  ScalarParameter() {defaultValue=1}
 ]
}
